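; Gameplay ability that freezes the pawn's current weapon in an ice block.
; On activation it records that weapon, and the server spawns B_WeaponIceBlock scaled from the mesh's imported bounds and attached to the weapon.
; Weapon-equip events add freeze tags and show the block only while that weapon is held, reversing both on switching away.
; Ending removes active tags, unhides the block, then server-side plays the thaw cue and kills the block.
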(event Ability|EventActivateAbility
  (bind _returnvalue (Ability|GetActivatingPawn))
  (bind _currentweapon (Variables|/Script/FortniteGame.FortPawn|GetCurrentWeapon _returnvalue))
  (Variables|Default|SetFrozenWeapon _currentweapon)
  (FortPawn|BindEventtoOnWeaponEquippedDelegate _returnvalue (delegate Custom|Event_OnWeaponEquipped))
  (CallFunction|EventonWeaponEquipped _currentweapon)
  (CallFunction|EncaseWeapon))

(event Custom|Event_OnWeaponEquipped
  (params (NewWeapon :type Fort%20Weapon%20Object%20Reference) (PrevWeapon :type Fort%20Weapon%20Object%20Reference))
  (bind _returnvalue (Ability|GetActivatingPawn))
  (bind _iceblock (Variables|Default|GetIceBlock))
  (bind _freezetags (Variables|Default|GetFreezeTags))
  (if (== NewWeapon (Variables|Default|GetFrozenWeapon))
    (if (not (Variables|Default|GetWasFrozen))
      (Ability|GameplayEffect|AddLooseGameplayTags _returnvalue _freezetags)
      (Variables|Default|SetWasFrozen true)
      (Rendering|SetActorHiddenInGame _iceblock false))
    (elif (Variables|Default|GetWasFrozen)
      (Ability|GameplayEffect|RemoveLooseGameplayTags _returnvalue _freezetags)
      (Variables|Default|SetWasFrozen false)
      (Rendering|SetActorHiddenInGame _iceblock true))))

(event Ability|EventOnEndAbility
  (params (bWasCancelled :type Boolean))
  (if (Variables|Default|GetWasFrozen)
    (Ability|GameplayEffect|RemoveLooseGameplayTags (Ability|GetActivatingPawn) (Variables|Default|GetFreezeTags))
    (else
      _))
  (Rendering|SetActorHiddenInGame (Variables|Default|GetIceBlock) false)
  (CallFunction|ThawWeapon))

(event Ability|EventFailedToActivatePassiveAbility
  (params (ActorInfo :type Gameplay%20Ability%20Actor%20Info%20Structure))
  (|CallParentFunction))

(fn EncaseWeapon
  (bind (_origin _boxextent) (Math|BoxSphereBounds|BreakBoxSphereBounds (Mesh|GetImportedBounds (Variables|/Script/Engine.SkinnedMeshComponent|GetSkeletalMesh (Variables|/Script/FortniteGame.FortWeapon|GetWeaponMesh (Variables|Default|GetFrozenWeapon))))))
  (bind _frozenweapon (Variables|Default|GetFrozenWeapon))
  (Utilities|IsServer
    (:true
      (Utilities|IsValid _frozenweapon
        (:Is%20Valid
          (bind _returnvalue (Game|SpawnActorfromClass "/Ice/Gameplay/IceGrenade/B_WeaponIceBlock.B_WeaponIceBlock_C" (Math|Transform|MakeTransform (Collision|GetComponentBounds (Variables|/Script/FortniteGame.FortWeapon|GetWeaponMesh _frozenweapon)) "0, 0, 0" (* _boxextent "1.000000,1.100000,1.000000")) "Undefined" "OverrideRootScale" 0 0 0))
          (bind _output_get (Variables|Default|SetIceBlock _returnvalue))
          (Transformation|AttachActorToComponent _output_get (Variables|/Script/Engine.Actor|GetRootComponent _frozenweapon) "None" "KeepWorld" "SnapToTarget" "KeepWorld" true))))))

(fn ThawWeapon
  (Utilities|IsServer
    (:true
      (GameplayCue|ExecuteGameplayCue (Ability|GetActivatingPawn) (Variables|Default|GetThawGameplayCue))
      (Gameplay|Die :self (Variables|Default|GetIceBlock) :!class /Script/FortniteGame.BuildingActor))))
